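# Physics asset for the male medium Scuba 01 player body cosmetic, covering only the upper spine, clavicles and selected dyn_hose bones.
# Large capsules cover the clavicles, while spine_05 and most hose points use zero-length, sphere-like capsules.
# Both clavicles constrain to spine_05; each hose body is angularly locked to its same-side clavicle or spine_05.
# The hose setup is asymmetric, with an extra tiny capsule on the right.
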PhysicsAsset /BRCosmetics/Characters/Player/Male/Medium/Bodies/M_MED_Scuba_01/Meshes/M_MED_Scuba_01_Physics
bodies {
 clavicle_l [{shapeType=Capsule, length=10, radius=8.5, center=<8.7678 0.4433 0.5895>, rotation=<rot 64.8568 -3.4987 -5.8926>}]
 clavicle_r [{shapeType=Capsule, length=10, radius=8.5, center=<-9.7727 -0.494 -0.6571>, rotation=<rot 64.8568 -3.4987 -5.8927>}]
 dyn_hose_l_03 [{shapeType=Capsule, length=0, radius=3}]
 dyn_hose_l_06 [{shapeType=Capsule, length=0, radius=3}]
 dyn_hose_r_03 [{shapeType=Capsule, length=0, radius=3}]
 dyn_hose_r_05 [{shapeType=Capsule, length=0.505, radius=0.505, rotation=<rot 0 -90 0>}]
 dyn_hose_r_07 [{shapeType=Capsule, length=0, radius=3}]
 spine_05 [{shapeType=Capsule, length=0, radius=5}]
}
constraints [
 {bone1Name=clavicle_l, bone2Name=spine_05, swing1LimitDegrees=45, swing2LimitDegrees=45, twistLimitDegrees=45}
 {
  bone1Name=clavicle_r
  bone2Name=spine_05
  swing1LimitDegrees=45
  swing1Motion=Free
  swing2LimitDegrees=45
  swing2Motion=Free
  twistLimitDegrees=45
  twistMotion=Free
 }
 {
  bone1Name=dyn_hose_l_06
  bone2Name=spine_05
  swing1LimitDegrees=45
  swing1Motion=Locked
  swing2LimitDegrees=45
  swing2Motion=Locked
  twistLimitDegrees=45
  twistMotion=Locked
 }
 {
  bone1Name=dyn_hose_r_07
  bone2Name=spine_05
  swing1LimitDegrees=45
  swing1Motion=Locked
  swing2LimitDegrees=45
  swing2Motion=Locked
  twistLimitDegrees=45
  twistMotion=Locked
 }
 {
  bone1Name=dyn_hose_r_03
  bone2Name=clavicle_r
  swing1LimitDegrees=45
  swing1Motion=Locked
  swing2LimitDegrees=45
  swing2Motion=Locked
  twistLimitDegrees=45
  twistMotion=Locked
 }
 {
  bone1Name=dyn_hose_l_03
  bone2Name=clavicle_l
  swing1LimitDegrees=45
  swing1Motion=Locked
  swing2LimitDegrees=45
  swing2Motion=Locked
  twistLimitDegrees=45
  twistMotion=Locked
 }
 {
  bone1Name=dyn_hose_r_05
  bone2Name=spine_05
  swing1LimitDegrees=45
  swing1Motion=Locked
  swing2LimitDegrees=45
  swing2Motion=Locked
  twistLimitDegrees=45
  twistMotion=Locked
 }
]
